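# Looping ambient sound cue for water overflowing from a pipe drain opening, classed as outdoor Athena ambience.
# Every quality level shares one chain: a modulator adds slight pitch variation and a fixed volume boost, then a short random delay leads to an equal-weight pick between medium and large looping waves.
# It uses sphere attenuation and overrides concurrency to allow few voices, stopping the lowest priority.
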SoundCue /Game/Sounds/Asteria/Environment_Actors/WaterPipeOverflow/Spot_Water_PipeOverflow_DrainOpening_All_Loop_Cue
root=QualityLevel_0
nodes [
 {id=Delay_1, children [Random_0], properties {delayMax=0.5}}
 {id=Modulator_1, children [Delay_1], properties {pitchMax=1.1, pitchMin=0.9, volumeMax=1.2, volumeMin=1.2}}
 {id=QualityLevel_0, children [Modulator_1,Modulator_1,Modulator_1,Modulator_1]}
 {id=Random_0, children [WavePlayer_0,WavePlayer_2], properties {weights [1,1]}}
 {
  id=WavePlayer_0
  properties {bLooping=true, soundWaveAssetPtr=&/Spot_Water_PipeOverflow_DrainOpening_Medium_Loop}
 }
 {
  id=WavePlayer_2
  properties {bLooping=true, soundWaveAssetPtr=&/Spot_Water_PipeOverflow_DrainOpening_Large_Loop}
 }
]
properties {
 attenuationSettings=&/Spot_Water_PipeOverflow_DrainOpening_Sphere_Att
 bOverrideConcurrency=true
 maxDistance=1348.9999
 priority=10
 soundClassObject=&Fort_Ambient_Outdoors_Athena
 volumeMultiplier=0.401
 concurrencyOverrides {maxCount=2, resolutionRule=StopLowestPriority, voiceStealReleaseTime=2}
}
