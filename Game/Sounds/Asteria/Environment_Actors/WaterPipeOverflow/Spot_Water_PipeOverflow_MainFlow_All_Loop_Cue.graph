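# Looping ambient sound cue for the main flow of an overflowing water pipe, using capsule attenuation and the Fort water interaction sound class.
# A random node picks the Medium or Large loop with equal weight, which passes through a three-second volume fade-in, a short random start delay and slight pitch variation.
# All quality levels share that chain.
# A concurrency override caps it at three voices, stopping lowest priority first.
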SoundCue /Game/Sounds/Asteria/Environment_Actors/WaterPipeOverflow/Spot_Water_PipeOverflow_MainFlow_All_Loop_Cue
root=QualityLevel_0
nodes [
 {id=Delay_1, children [Enveloper_1], properties {delayMax=0.5}}
 {
  id=Enveloper_1
  children [Random_0]
  properties {
   pitchCurve {editorCurveData {keys [{time=0}]}}
   volumeCurve {editorCurveData {keys [<at -0.0001 0 RCIM_Cubic>, {arriveTangent=0.4839, interpMode=RCIM_Cubic, leaveTangent=0.4839, tangentMode=RCTM_User, time=3, value=1}]}}
  }
 }
 {id=Modulator_1, children [Delay_1], properties {pitchMax=1.1, pitchMin=0.9, volumeMax=1, volumeMin=1}}
 {id=QualityLevel_0, children [Modulator_1,Modulator_1,Modulator_1,Modulator_1]}
 {id=Random_0, children [WavePlayer_1,WavePlayer_3], properties {weights [1,1]}}
 {
  id=WavePlayer_1
  properties {bLooping=true, soundWaveAssetPtr=&/Spot_Water_PipeOverflow_MainFlow_Medium_Loop}
 }
 {
  id=WavePlayer_3
  properties {bLooping=true, soundWaveAssetPtr=&/Spot_Water_PipeOverflow_MainFlow_Large_Loop}
 }
]
properties {
 attenuationSettings=&/Spot_Water_PipeOverflow_MainFlow_Capsule_Att
 bOverrideConcurrency=true
 maxDistance=3150
 priority=10
 soundClassObject=/Game/Sounds/FortSoundClassesAndModes/Fort_Water_Interaction
 volumeMultiplier=0.8
 concurrencyOverrides {maxCount=3, resolutionRule=StopLowestPriority, voiceStealReleaseTime=3}
}
