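# Looping ambient sound cue for the medium main flow of an overflowing water pipe, routed to the outdoor ambient sound class.
# A looping wave fades in through a volume envelope, starts after a short random delay and gets slight random pitch variation.
# All quality-level inputs share that one chain.
# Capsule-named attenuation shapes falloff, and a concurrency override caps simultaneous instances, stopping the lowest priority.
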SoundCue /Game/Sounds/Asteria/Environment_Actors/WaterPipeOverflow/Spot_Water_PipeOverflow_MainFlow_Medium_Loop_Cue
root=QualityLevel_0
nodes [
 {id=Delay_1, children [Enveloper_1], properties {delayMax=1.5}}
 {
  id=Enveloper_1
  children [WavePlayer_0]
  properties {
   pitchCurve {editorCurveData {keys [{time=0}]}}
   volumeCurve {editorCurveData {keys [<at -0.0001 0 RCIM_Cubic>, {arriveTangent=0.4839, interpMode=RCIM_Cubic, leaveTangent=0.4839, tangentMode=RCTM_User, time=3, value=1}]}}
  }
 }
 {id=Modulator_1, children [Delay_1], properties {pitchMax=1.1, pitchMin=0.9, volumeMax=1, volumeMin=1}}
 {id=QualityLevel_0, children [Modulator_1,Modulator_1,Modulator_1,Modulator_1]}
 {
  id=WavePlayer_0
  properties {bLooping=true, soundWaveAssetPtr=&/Spot_Water_PipeOverflow_MainFlow_Medium_Loop}
 }
]
properties {
 attenuationSettings=&/Spot_Water_PipeOverflow_MainFlow_Capsule_Att
 bOverrideConcurrency=true
 maxDistance=3150
 priority=10
 soundClassObject=&Fort_Ambient_Outdoors_Athena
 volumeMultiplier=0.501
 concurrencyOverrides {maxCount=2, resolutionRule=StopLowestPriority}
}
